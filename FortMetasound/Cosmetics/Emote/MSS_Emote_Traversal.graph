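# Stereo looping MetaSound source for traversal emotes, blending idle and forward-movement audio by emote move speed.
# Interpolated EmoteMoveSpeed crossfades two MSS_Emote_Base instances and maps to a pitch multiplier shared by both.
# Two MSP_OneshotGate patches watch that speed, firing random Forward or Backward (stop) one-shots with retrigger delay and ADSR-controlled gain.
# A final mixer sums the crossfaded loops and both one-shots, with separate one-shot volume inputs.
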MetaSoundSource /FortMetasound/Cosmetics/Emote/MSS_Emote_Traversal
inputs [
 Backward.Sounds {type=WaveAsset:Array, access=Constructor}
 DelayLoopByStart {type=Bool, access=Constructor}
 EmoteMoveSpeed {type=Float}
 FadeInTime {type=Time, access=Constructor}
 Forward.Loop.Sounds {type=WaveAsset:Array, access=Constructor}
 Forward.LoopVolume {type=Float, access=Constructor}
 Forward.OneshotThreshold {type=Float, access=Constructor}
 Forward.OneShotVolume {type=Float, access=Constructor}
 Forward.Pitch {type=Float, access=Constructor}
 Forward.Sounds {type=WaveAsset:Array, access=Constructor}
 Forward.StartVolume {type=Float, access=Constructor}
 Idle.LoopVolume {type=Float, access=Constructor}
 Idle.StartVolume {type=Float, access=Constructor}
 InterpTime {type=Time, access=Constructor}
 Looping {type=Bool, access=Constructor}
 OneshotRetriggerDelay {type=Time, access=Constructor}
 PitchMultiplier {type=Float, access=Constructor}
 Stop.OneshotThreshold {type=Float, access=Constructor}
 Stop.OneShotVolume {type=Float, access=Constructor}
 Traversal.Loop.Sounds {type=WaveAsset:Array, access=Constructor}
 Traversal.Sounds {type=WaveAsset:Array, access=Constructor}
 UE.Source.OnPlay {type=Trigger}
 UE.Source.StartTime.StartTime {type=Time}
]
interfaces [UE.OutputFormat.Stereo:1.0, UE.Source.StartTime:1.0, UE.Source:1.1]
outputs [
 "UE.OutputFormat.Stereo.Audio:0" {type=Audio}
 "UE.OutputFormat.Stereo.Audio:1" {type=Audio}
]
pages [Default {
  input_defaults {
   DelayLoopByStart=null
   EmoteMoveSpeed=0
   FadeInTime=0.01
   Forward.LoopVolume=1
   Forward.OneShotVolume=1
   Forward.OneshotThreshold=0.2
   Forward.Pitch=1.2
   Forward.StartVolume=1
   Idle.LoopVolume=1
   Idle.StartVolume=1
   InterpTime=0.25
   Looping="<MetaSoundBuilderResult.SUCCEEDED: 0>"
   OneshotRetriggerDelay=2
   PitchMultiplier=1
   Stop.OneShotVolume=1
   Stop.OneshotThreshold=0.8
   UE.Source.OnPlay=null
   UE.Source.StartTime.StartTime=null
   Backward.Sounds []
   Forward.Loop.Sounds []
   Forward.Sounds []
   Traversal.Loop.Sounds []
   Traversal.Sounds []
  }
  nodes [
   n1 = VariableAccessor.Float(Variable=VariableMutator.Float(Variable=InitVariable.Float().Variable, Value=UE.InterpTo:Audio(Target=^EmoteMoveSpeed, Interp Time=^InterpTime).Value).Variable)
   n2 = VariableAccessor.Float(Variable=@n1.Variable)
   n3 = VariableAccessor.Float(Variable=@n2.Variable)
   n4 = &MSP_OneshotGate(OneshotThreshold=^Forward.OneshotThreshold, RetriggerDelay=^OneshotRetriggerDelay, Value=@n3.Value)
   ADSREnvelopeADSREnvelope_0 = ADSR Envelope.ADSR Envelope:Float(Trigger Attack=@n4.OnGateEnter, Trigger Release=@n4.OnGateExit) {also {defaults {Attack Time=0, Sustain Level=1}}}
   ArrayRandomGet_1 = Array.Random Get:WaveAsset:Array(In Array=^Forward.Sounds, Next=@n4.OnGateEnter)
   WavePlayer_0 = UE.Wave Player:Stereo(Play=@ArrayRandomGet_1.On Next, Wave Asset=@ArrayRandomGet_1.Value) {also {version [1,1]}}
   AudioMixerAudioMixerStereo2_0 = "AudioMixer.Audio Mixer (Stereo, 2)"(Gain 0=@ADSREnvelopeADSREnvelope_0.Out Envelope, In 0 L=@WavePlayer_0.Out Left, In 0 R=@WavePlayer_0.Out Right)
   n5 = &MSP_OneshotGate(RetriggerDelay=^OneshotRetriggerDelay, OneshotThreshold=^Stop.OneshotThreshold, Value=VariableAccessor.Float(Variable=@n3.Variable).Value) {also {defaults {EnterCondition=4, ExitCondition=3}}}
   ADSREnvelopeADSREnvelope_1 = ADSR Envelope.ADSR Envelope:Float(Trigger Attack=@n5.OnGateEnter, Trigger Release=@n5.OnGateExit) {also {defaults {Attack Time=0, Sustain Level=1}}}
   ArrayRandomGet_0 = Array.Random Get:WaveAsset:Array(In Array=^Backward.Sounds, Next=@n5.OnGateEnter)
   WavePlayer_1 = UE.Wave Player:Stereo(Play=@ArrayRandomGet_0.On Next, Wave Asset=@ArrayRandomGet_0.Value) {also {version [1,1]}}
   AudioMixerAudioMixerStereo2_1 = "AudioMixer.Audio Mixer (Stereo, 2)"(Gain 0=@ADSREnvelopeADSREnvelope_1.Out Envelope, In 0 L=@WavePlayer_1.Out Left, In 0 R=@WavePlayer_1.Out Right)
   n6 = MapRange.MapRange:Float(Out Range B=^Forward.Pitch, In=@n2.Value) {also {defaults {Out Range A=1}}}
   n7 = VariableAccessor.Float(Variable=VariableMutator.Float(Variable=InitVariable.Float().Variable, Value=UE.Multiply:Float(PrimaryOperand=^PitchMultiplier, AdditionalOperands=@n6.Out Value).Out).Variable)
   n8 = &/MSS_Emote_Base(
    DelayLoopByStart=^DelayLoopByStart
    FadeInTime=^FadeInTime
    Loop.Sounds=^Forward.Loop.Sounds
    LoopVolumeMultiplier=^Forward.LoopVolume
    IntroVolumeMultiplier=^Forward.StartVolume
    UE.Source.OnPlay=^UE.Source.OnPlay
    UE.Source.StartTime.StartTime=^UE.Source.StartTime.StartTime
    PitchTemplateMultiplier=VariableAccessor.Float(Variable=@n7.Variable).Value
   ) {also {defaults {Looping="<MetaSoundBuilderResult.SUCCEEDED: 0>"}}}
   n9 = &MSS_Emote_Base(
    DelayLoopByStart=^DelayLoopByStart
    FadeInTime=^FadeInTime
    LoopVolumeMultiplier=^Idle.LoopVolume
    IntroVolumeMultiplier=^Idle.StartVolume
    Looping=^Looping
    Loop.Sounds=^Traversal.Loop.Sounds
    Intro.Sounds=^Traversal.Sounds
    UE.Source.OnPlay=^UE.Source.OnPlay
    UE.Source.StartTime.StartTime=^UE.Source.StartTime.StartTime
    PitchTemplateMultiplier=@n7.Value
   )
   CrossfadeTriggerRouteAudio2_0 = "Crossfade.Trigger Route (Audio, 2):Audio"(In 1=@n8."UE.OutputFormat.Stereo.Audio:1", In 0=@n9."UE.OutputFormat.Stereo.Audio:1", Crossfade Value=@n1.Value)
   CrossfadeTriggerRouteAudio2_1 = "Crossfade.Trigger Route (Audio, 2):Audio"(In 1=@n8."UE.OutputFormat.Stereo.Audio:0", In 0=@n9."UE.OutputFormat.Stereo.Audio:0", Crossfade Value=@n1.Value)
   AudioMixerAudioMixerStereo8_0 = "AudioMixer.Audio Mixer (Stereo, 8)"(
    Gain 1=^Forward.OneShotVolume
    Gain 2=^Stop.OneShotVolume
    In 1 L=@AudioMixerAudioMixerStereo2_0.Out L
    In 1 R=@AudioMixerAudioMixerStereo2_0.Out R
    In 2 L=@AudioMixerAudioMixerStereo2_1.Out L
    In 2 R=@AudioMixerAudioMixerStereo2_1.Out R
    In 0 R=@CrossfadeTriggerRouteAudio2_0.Out
    In 0 L=@CrossfadeTriggerRouteAudio2_1.Out
   ) {also {defaults {Gain 3=0.75}}}
   InitVariable.Time()
   InitVariable.Time()
  ]
  links [
   {dst="@UE.OutputFormat.Stereo.Audio:0", src=AudioMixerAudioMixerStereo8_0.Out L}
   {dst="@UE.OutputFormat.Stereo.Audio:1", src=AudioMixerAudioMixerStereo8_0.Out R}
  ]
 }]
properties {
 assetImportData=&self:AssetImportData
 attenuationSettings=&Fort_Emotes_Athena_Att
 audioPropertiesSheet=/BRCosmetics/Sounds/Settings/PropertySheets/Emotes/APS_Cosmetic_Emote_Traversal
 bLooping=true
 duration=10000
 maxDistance=1600
 numChannels=2
 outputFormat=Stereo
 sampleRate=48000
 soundClassObject=&Fort_Music_Emote
 concurrencySet [/Game/Sounds/Concurrency_Rules/Athena_EmoteTraversal_Conc, /Game/Sounds/Concurrency_Rules/Athena_EmoteFoley_Global_Conc]
}
