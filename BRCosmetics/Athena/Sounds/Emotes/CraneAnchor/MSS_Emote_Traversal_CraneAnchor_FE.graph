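# Stereo looping MetaSound source for the CraneAnchor traversal emote, playing in the emote music sound class with emote attenuation and foley concurrency.
# It wraps the shared MSS_Emote_Traversal graph, forwarding EmoteMoveSpeed and tuning inputs and routing its stereo output straight out.
# The CraneAnchor loop fills both forward and traversal loop slots, quieter at idle, while the other sound arrays are empty.
# PitchMultiplier and DelayLoopByStart are exposed but left unconnected.
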MetaSoundSource /BRCosmetics/Athena/Sounds/Emotes/CraneAnchor/MSS_Emote_Traversal_CraneAnchor_FE
inputs [
 Backward.Sounds {type=WaveAsset:Array, access=Constructor}
 CopyrightedAudio {type=Bool, access=Constructor}
 DelayLoopByStart {type=Bool, access=Constructor}
 EmoteMoveSpeed {type=Float}
 FadeInTime {type=Time, access=Constructor}
 Forward.Loop.Sounds {type=WaveAsset:Array, access=Constructor}
 Forward.LoopVolume {type=Float, access=Constructor}
 Forward.OneshotThreshold {type=Float, access=Constructor}
 Forward.OneShotVolume {type=Float, access=Constructor}
 Forward.Pitch {type=Float, access=Constructor}
 Forward.Sounds {type=WaveAsset:Array, access=Constructor}
 Forward.StartVolume {type=Float, access=Constructor}
 Idle.LoopVolume {type=Float, access=Constructor}
 Idle.StartVolume {type=Float, access=Constructor}
 InterpTime {type=Time, access=Constructor}
 Looping {type=Bool, access=Constructor}
 OneshotRetriggerDelay {type=Time, access=Constructor}
 PitchMultiplier {type=Float, access=Constructor}
 Stop.OneshotThreshold {type=Float, access=Constructor}
 Stop.OneShotVolume {type=Float, access=Constructor}
 Traversal.Loop.Sounds {type=WaveAsset:Array, access=Constructor}
 Traversal.Sounds {type=WaveAsset:Array, access=Constructor}
 UE.Source.OnPlay {type=Trigger}
]
interfaces [UE.OutputFormat.Stereo:1.0, UE.Source:1.1]
outputs [
 "UE.OutputFormat.Stereo.Audio:0" {type=Audio}
 "UE.OutputFormat.Stereo.Audio:1" {type=Audio}
]
pages [Default {
  input_defaults {
   CopyrightedAudio=null
   DelayLoopByStart=null
   EmoteMoveSpeed=0
   FadeInTime=0.01
   Forward.LoopVolume=1
   Forward.OneShotVolume=1
   Forward.OneshotThreshold=0.2
   Forward.Pitch=1
   Forward.StartVolume=1
   Idle.LoopVolume=0.4
   Idle.StartVolume=0.2
   InterpTime=0.25
   Looping="<MetaSoundBuilderResult.SUCCEEDED: 0>"
   OneshotRetriggerDelay=2
   PitchMultiplier=1
   Stop.OneShotVolume=1
   Stop.OneshotThreshold=0.8
   UE.Source.OnPlay=null
   Backward.Sounds []
   Forward.Loop.Sounds [&/Emote_CraneAnchor_Loop]
   Forward.Sounds []
   Traversal.Loop.Sounds [&Emote_CraneAnchor_Loop]
   Traversal.Sounds []
  }
  nodes [MSS_Emote_Traversal_0 = /FortMetasound/Cosmetics/Emote/MSS_Emote_Traversal(
    Backward.Sounds=^Backward.Sounds
    CopyrightedAudio=^CopyrightedAudio
    EmoteMoveSpeed=^EmoteMoveSpeed
    FadeInTime=^FadeInTime
    Forward.Loop.Sounds=^Forward.Loop.Sounds
    Forward.LoopVolume=^Forward.LoopVolume
    Forward.OneShotVolume=^Forward.OneShotVolume
    Forward.OneshotThreshold=^Forward.OneshotThreshold
    Forward.Pitch=^Forward.Pitch
    Forward.Sounds=^Forward.Sounds
    Forward.StartVolume=^Forward.StartVolume
    Idle.LoopVolume=^Idle.LoopVolume
    Idle.StartVolume=^Idle.StartVolume
    InterpTime=^InterpTime
    Looping=^Looping
    OneshotRetriggerDelay=^OneshotRetriggerDelay
    Stop.OneShotVolume=^Stop.OneShotVolume
    Stop.OneshotThreshold=^Stop.OneshotThreshold
    Traversal.Loop.Sounds=^Traversal.Loop.Sounds
    Traversal.Sounds=^Traversal.Sounds
    UE.Source.OnPlay=^UE.Source.OnPlay
   )]
  links [
   {dst="@UE.OutputFormat.Stereo.Audio:0", src=MSS_Emote_Traversal_0.UE.OutputFormat.Stereo.Audio:0}
   {dst="@UE.OutputFormat.Stereo.Audio:1", src=MSS_Emote_Traversal_0.UE.OutputFormat.Stereo.Audio:1}
  ]
 }]
properties {
 assetImportData=&self:AssetImportData
 attenuationSettings=&Fort_Emotes_Athena_Att
 bLooping=true
 duration=10000
 maxDistance=1600
 numChannels=2
 outputFormat=Stereo
 sampleRate=48000
 soundClassObject=&Fort_Music_Emote
 volume=3.5
 concurrencySet [/Game/Sounds/Concurrency_Rules/Athena_EmoteFoley_Conc]
 soundSubmixSends [{maxSendLevel=1.5, sendLevel=1.5}]
}
